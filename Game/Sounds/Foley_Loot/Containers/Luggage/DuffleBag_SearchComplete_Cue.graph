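# Sound cue for completing a duffle bag search, classed as loot-collect foley with prop-search attenuation.
# The first two quality levels randomly pick one of two search recordings, shape it with a fast-attack, sub-second fade volume envelope, then slightly vary pitch and volume.
# The other two levels play a single optimized search-end wave directly.
# A concurrency override caps simultaneous instances, stopping the farthest first.
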SoundCue /Game/Sounds/Foley_Loot/Containers/Luggage/DuffleBag_SearchComplete_Cue
root=QualityLevel_0
nodes [
 {
  id=Enveloper_16
  children [Random_7]
  properties {
   volumeMax=1.2
   volumeMin=1.2
   pitchCurve {editorCurveData {keys [{time=0}]}}
   volumeCurve {editorCurveData {keys [<at 0.0017 0.0028>,<at 0.0237 0.7156>,<at 0.34 0.434>,<at 0.747 0>]}}
  }
 }
 {id=Modulator_15, children [Enveloper_16], properties {pitchMax=1.07, pitchMin=0.92, volumeMax=1.07, volumeMin=0.92}}
 {id=QualityLevel_0, children [Modulator_15,Modulator_15,WavePlayer_0,WavePlayer_0]}
 {id=Random_7, children [WavePlayer_39, WavePlayer_40], properties {weights [1,1]}}
 {id=WavePlayer_0, properties {soundWaveAssetPtr=&/DuffleBag_SearchEnd_opt}}
 {id=WavePlayer_39, properties {soundWaveAssetPtr=&/DuffleBag_Search03}}
 {id=WavePlayer_40, properties {soundWaveAssetPtr=&/DuffleBag_Search05}}
]
properties {
 attenuationSettings=/Game/Sounds/Attenuation/Fort_Prop_Search_Att
 bOverrideConcurrency=true
 maxDistance=1300
 soundClassObject=&Fort_Loot_Collect
 volumeMultiplier=0.7
 concurrencyOverrides {maxCount=4, resolutionRule=StopFarthestThenPreventNew}
}
