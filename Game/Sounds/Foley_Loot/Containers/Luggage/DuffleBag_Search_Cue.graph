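# Sound cue for searching a duffle bag loot container, with prop-search attenuation and a concurrency override that prevents new instances at the limit.
# A quality-level root selects among four progressively simpler branches, ending in a single optimized wave with and then without pitch/volume modulation.
# The two richer branches mix randomized, modulated bag rustles, a one-second-delayed reuse of that chain, and quieter random gear foley, differing in variation count.
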SoundCue /Game/Sounds/Foley_Loot/Containers/Luggage/DuffleBag_Search_Cue
root=QualityLevel_0
nodes [
 {id=Delay_0, children [Modulator_0], properties {delayMax=1, delayMin=1}}
 {id=Delay_6, children [Modulator_23], properties {delayMax=1, delayMin=1}}
 {id=Mixer_0, children [Modulator_0,Delay_0,Random_1], properties {inputVolume [1,0.8,0.5]}}
 {id=Mixer_4, children [Modulator_23,Delay_6,Random_5], properties {inputVolume [1,0.8,0.5]}}
 {id=Modulator_0, children [Random_0], properties {pitchMax=1.07, pitchMin=0.92, volumeMax=1.07, volumeMin=0.92}}
 {id=Modulator_1, children [WavePlayer_0], properties {pitchMax=1.07, pitchMin=0.92, volumeMax=1.07, volumeMin=0.92}}
 {id=Modulator_23, children [Random_6], properties {pitchMax=1.07, pitchMin=0.92, volumeMax=1.07, volumeMin=0.92}}
 {id=QualityLevel_0, children [Mixer_4,Mixer_0,Modulator_1,WavePlayer_5]}
 {id=Random_0, children [WavePlayer_1,WavePlayer_2], properties {weights [1,1]}}
 {id=Random_1, children [WavePlayer_3,WavePlayer_4], properties {weights [1,1]}}
 {id=Random_5, children [WavePlayer_20, WavePlayer_21, WavePlayer_22], properties {weights [1,1,0.5]}}
 {id=Random_6, children [WavePlayer_34, WavePlayer_35, WavePlayer_36, WavePlayer_37], properties {weights [1,1,1,1]}}
 {id=WavePlayer_0, properties {soundWaveAssetPtr=&/DuffleBag_Search_opt}}
 {id=WavePlayer_1, properties {soundWaveAssetPtr=&/DuffleBag_Search12}}
 {id=WavePlayer_2, properties {soundWaveAssetPtr=&/DuffleBag_Search03}}
 {id=WavePlayer_20, properties {soundWaveAssetPtr=/Game/Sounds/Foley_Loot/Containers/Misc/Prop_Gear_Search_05}}
 {id=WavePlayer_21, properties {soundWaveAssetPtr=/Game/Sounds/Foley_Loot/Containers/Misc/Prop_Gear_Search_06}}
 {id=WavePlayer_22, properties {soundWaveAssetPtr=/Game/Sounds/Foley_Loot/Containers/Misc/Prop_Gear_Search_01}}
 {id=WavePlayer_3, properties {soundWaveAssetPtr=&Prop_Gear_Search_05}}
 {id=WavePlayer_34, properties {soundWaveAssetPtr=&DuffleBag_Search12}}
 {id=WavePlayer_35, properties {soundWaveAssetPtr=&DuffleBag_Search03}}
 {id=WavePlayer_36, properties {soundWaveAssetPtr=&/DuffleBag_Search05}}
 {id=WavePlayer_37, properties {soundWaveAssetPtr=&/DuffleBag_Search08}}
 {id=WavePlayer_4, properties {soundWaveAssetPtr=&Prop_Gear_Search_06}}
 {id=WavePlayer_5, properties {soundWaveAssetPtr=&DuffleBag_Search_opt}}
]
properties {
 attenuationSettings=/Game/Sounds/Attenuation/Fort_Prop_Search_Att
 bOverrideConcurrency=true
 maxDistance=1300
 soundClassObject=&Fort_Loot_Collect
 volumeMultiplier=1
 concurrencyOverrides {resolutionRule=PreventNew}
}
