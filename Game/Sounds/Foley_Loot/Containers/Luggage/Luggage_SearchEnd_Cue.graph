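# Sound cue for the end of searching a luggage loot container, pairing luggage search foley with the no-more-loot UI sound.
# One of three equally weighted search waves is picked at random, faded out by an Enveloper and slightly varied in pitch and volume by a Modulator.
# A Mixer layers that foley with the UI wave.
# The cue uses prop-search attenuation, the loot-collect sound class and its own concurrency limit favoring nearer instances.
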SoundCue /Game/Sounds/Foley_Loot/Containers/Luggage/Luggage_SearchEnd_Cue
root=Mixer_12
nodes [
 {
  id=Enveloper_13
  children [Random_4]
  properties {pitchCurve {editorCurveData {keys [{time=0}]}}, volumeCurve {editorCurveData {keys [<at 0 1>,<at 0.63 0>]}}}
 }
 {id=Mixer_12, children [WavePlayer_43, Modulator_15], properties {inputVolume [0.8,1.1]}}
 {id=Modulator_15, children [Enveloper_13], properties {pitchMax=1.07, pitchMin=0.92, volumeMax=1.07, volumeMin=0.92}}
 {id=Random_4, children [WavePlayer_27, WavePlayer_28, WavePlayer_29], properties {weights [1,1,1]}}
 {id=WavePlayer_27, properties {soundWaveAssetPtr=&/Luggage_Search01}}
 {id=WavePlayer_28, properties {soundWaveAssetPtr=&/Luggage_Search02}}
 {id=WavePlayer_29, properties {soundWaveAssetPtr=&/Luggage_Search03}}
 {id=WavePlayer_43, properties {soundWaveAssetPtr=/Game/Sounds/Fort_UI_Sounds/Gameplay/Fort_UI_NoMoreLoot}}
]
properties {
 attenuationSettings=/Game/Sounds/Attenuation/Fort_Prop_Search_Att
 bOverrideConcurrency=true
 maxDistance=1300
 soundClassObject=&Fort_Loot_Collect
 volumeMultiplier=0.9
 concurrencyOverrides {maxCount=4, resolutionRule=StopFarthestThenPreventNew}
}
