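# Sound cue for searching luggage loot containers, with prop-search attenuation and a prevent-new concurrency override.
# A quality-level root selects between two layered mixes of differing variety and two single optimized-wave branches.
# Each mix plays random, pitch- and volume-modulated luggage searches, reusing that branch through a roughly one-second delay for a second pass.
# A gear-rustle layer joins them, randomized in the fuller mix and a single wave in the leaner one.
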SoundCue /Game/Sounds/Foley_Loot/Containers/Luggage/Luggage_Search_Cue
root=QualityLevel_0
nodes [
 {id=Delay_0, children [Modulator_0], properties {delayMax=1.2, delayMin=1.1}}
 {id=Delay_30, children [Modulator_23], properties {delayMax=1.2, delayMin=1.1}}
 {id=Mixer_0, children [Modulator_0,Delay_0,Modulator_1], properties {inputVolume [1.2,1,1]}}
 {id=Mixer_13, children [Modulator_23,Delay_30,Random_4], properties {inputVolume [1.2,1,1]}}
 {id=Modulator_0, children [Random_0], properties {pitchMax=1.07, pitchMin=0.92, volumeMax=1.07, volumeMin=0.92}}
 {id=Modulator_1, children [WavePlayer_3]}
 {id=Modulator_23, children [Random_3], properties {pitchMax=1.07, pitchMin=0.92, volumeMax=1.07, volumeMin=0.92}}
 {id=Modulator_3, children [WavePlayer_4]}
 {id=Modulator_4, children [WavePlayer_5]}
 {id=QualityLevel_0, children [Mixer_13,Mixer_0,Modulator_3,Modulator_4]}
 {id=Random_0, children [WavePlayer_0,WavePlayer_1], properties {weights [1,1]}}
 {id=Random_3, children [WavePlayer_24, WavePlayer_25, WavePlayer_26], properties {weights [1,1,1]}}
 {id=Random_4, children [WavePlayer_17, WavePlayer_18, WavePlayer_19], properties {weights [1,1,0.5]}}
 {id=WavePlayer_0, properties {soundWaveAssetPtr=&/Luggage_Search01}}
 {id=WavePlayer_1, properties {soundWaveAssetPtr=&/Luggage_Search02}}
 {id=WavePlayer_17, properties {soundWaveAssetPtr=/Game/Sounds/Foley_Loot/Containers/Misc/Prop_Gear_Search_05}}
 {id=WavePlayer_18, properties {soundWaveAssetPtr=/Game/Sounds/Foley_Loot/Containers/Misc/Prop_Gear_Search_06}}
 {id=WavePlayer_19, properties {soundWaveAssetPtr=/Game/Sounds/Foley_Loot/Containers/Misc/Prop_Gear_Search_01}}
 {id=WavePlayer_24, properties {soundWaveAssetPtr=&Luggage_Search01}}
 {id=WavePlayer_25, properties {soundWaveAssetPtr=&Luggage_Search02}}
 {id=WavePlayer_26, properties {soundWaveAssetPtr=&/Luggage_Search03}}
 {id=WavePlayer_3, properties {soundWaveAssetPtr=&Prop_Gear_Search_05}}
 {id=WavePlayer_4, properties {soundWaveAssetPtr=&/Luggage_Search_opt}}
 {id=WavePlayer_5, properties {soundWaveAssetPtr=&Luggage_Search_opt}}
]
properties {
 attenuationSettings=/Game/Sounds/Attenuation/Fort_Prop_Search_Att
 bOverrideConcurrency=true
 maxDistance=1300
 soundClassObject=&Fort_Loot_Collect
 volumeMultiplier=0.9
 concurrencyOverrides {resolutionRule=PreventNew}
}
